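# Stereo MetaSound source in the art QA audio folder that plays a looping music test file, apparently for testing audio reactivity.
# On play, a Wave Player starts the loop at a random time offset, so the starting point can vary.
# The player's On Play trigger fires an ADSR envelope that multiplies both channels, fading the audio in to full sustain before it reaches the outputs.
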MetaSoundSource /Game/QA/Art_QA/BR_Audio/Audio/MSS_BR_Audio_AudioReactivityTester
inputs [UE.Source.OnPlay {type=Trigger}]
interfaces [UE.OutputFormat.Stereo:1.0, UE.Source:1.1]
outputs [
 "UE.OutputFormat.Stereo.Audio:0" {type=Audio}
 "UE.OutputFormat.Stereo.Audio:1" {type=Audio}
]
pages [Default {
  input_defaults {UE.Source.OnPlay=null}
  nodes [
   WavePlayer_0 = UE.Wave Player:Stereo(Play=^UE.Source.OnPlay, Start Time=UE.RandomTime() {also {defaults {Max=126}}}.Value) {also {
     defaults {Loop="<MetaSoundBuilderResult.SUCCEEDED: 0>", Wave Asset=&/MX_AudioAnalysisTestFile_Loop}
     version [1,1]
    }}
   ADSREnvelopeADSREnvelope_0 = ADSR Envelope.ADSR Envelope:Audio(Trigger Attack=@WavePlayer_0.On Play) {also {defaults {Attack Time=1, Hard Reset="<MetaSoundBuilderResult.SUCCEEDED: 0>", Sustain Level=1}}}
   Multiply_0 = UE.Multiply:Audio(AdditionalOperands=@ADSREnvelopeADSREnvelope_0.Out Envelope, PrimaryOperand=@WavePlayer_0.Out Left)
   Multiply_1 = UE.Multiply:Audio(AdditionalOperands=@ADSREnvelopeADSREnvelope_0.Out Envelope, PrimaryOperand=@WavePlayer_0.Out Right)
  ]
  links [{dst="@UE.OutputFormat.Stereo.Audio:0", src=Multiply_0.Out}, {dst="@UE.OutputFormat.Stereo.Audio:1", src=Multiply_1.Out}]
 }]
properties {
 assetImportData=&self:AssetImportData
 bLooping=true
 duration=10000
 numChannels=2
 outputFormat=Stereo
 sampleRate=48000
 soundSubmixSends [{}]
}
