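# QA test stereo MetaSound source, named for audio modulation testing, that plays a voice test clip under the Voice sound class with NPC alert VO attenuation.
# On play it triggers a basic stereo wave player patch fed by the Wave Asset array, routed straight to the left and right outputs.
# A pre-effect send feeds the music volume sidechain audio bus, with distance limits set on the send.
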MetaSoundSource /Game/QA/Art_QA/BR_Audio/Audio/MSS_BR_Audio_Modulation_Test
inputs [
 UE.Source.OnPlay {type=Trigger}
 Wave Asset {type=WaveAsset:Array}
]
interfaces [UE.OutputFormat.Stereo:1.0, UE.Source:1.1]
outputs [
 "UE.OutputFormat.Stereo.Audio:0" {type=Audio}
 "UE.OutputFormat.Stereo.Audio:1" {type=Audio}
]
pages [Default {
  input_defaults {UE.Source.OnPlay=null, Wave Asset [&/VO_Test_One_Two]}
  nodes [MSP_Audio_WavePlayer_Basic_Stereo_0 = /FortMetasound/Patches/MSP_Audio_WavePlayer_Basic_Stereo(Play Trigger=^UE.Source.OnPlay, Wave Asset=^Wave Asset) {also {defaults {Is Looping?="<MetaSoundBuilderResult.SUCCEEDED: 0>"}}}]
  links [
   {dst="@UE.OutputFormat.Stereo.Audio:0", src=MSP_Audio_WavePlayer_Basic_Stereo_0.Out L}
   {dst="@UE.OutputFormat.Stereo.Audio:1", src=MSP_Audio_WavePlayer_Basic_Stereo_0.Out R}
  ]
 }]
properties {
 assetImportData=&self:AssetImportData
 attenuationSettings=/Game/Sounds/Fort_Human_VO_BR/NPCs/Shared/NPC_VO_Alerted_Att
 bLooping=true
 duration=10000
 maxDistance=12500
 numChannels=2
 outputFormat=Stereo
 sampleRate=48000
 soundClassObject=/Game/Sounds/FortSoundClassesAndModes/Voice
 preEffectBusSends [{
   audioBus=&AB_Sidechain_FortMusic_Vol
   maxSendDistance=1000
   minSendDistance=100
   minSendLevel=0
   sendLevel=1
  }]
}
