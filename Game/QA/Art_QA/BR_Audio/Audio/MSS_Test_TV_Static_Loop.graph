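# Test stereo MetaSound Source in the QA audio folder that plays TV static tuning as a quiet, attenuated ambient prop sound.
# On Play triggers a single stereo Wave Player, whose left and right outputs feed the source outputs directly.
# The source is flagged as looping, but the player's Loop default holds a builder result string rather than a plain boolean.
# A test TV attenuation asset and the Athena ambient props sound class govern playback.
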MetaSoundSource /Game/QA/Art_QA/BR_Audio/Audio/MSS_Test_TV_Static_Loop
inputs [UE.Source.OnPlay {type=Trigger}]
interfaces [UE.OutputFormat.Stereo:1.0, UE.Source:1.1]
outputs [
 "UE.OutputFormat.Stereo.Audio:0" {type=Audio}
 "UE.OutputFormat.Stereo.Audio:1" {type=Audio}
]
pages [Default {
  input_defaults {UE.Source.OnPlay=null}
  nodes [WavePlayer_0 = UE.Wave Player:Stereo(Play=^UE.Source.OnPlay) {also {defaults {Loop="<MetaSoundBuilderResult.SUCCEEDED: 0>", Wave Asset=&/TV_Static_Tuning_Loop}, version [1,1]}}]
  links [{dst="@UE.OutputFormat.Stereo.Audio:0", src=WavePlayer_0.Out Left}, {dst="@UE.OutputFormat.Stereo.Audio:1", src=WavePlayer_0.Out Right}]
 }]
properties {
 assetImportData=&self:AssetImportData
 attenuationSettings=&/ATT_Test_TV_Loop
 bLooping=true
 duration=10000
 maxDistance=4200
 numChannels=2
 outputFormat=Stereo
 sampleRate=48000
 soundClassObject=&Fort_Ambient_Props_Athena
 volume=0.1
}
