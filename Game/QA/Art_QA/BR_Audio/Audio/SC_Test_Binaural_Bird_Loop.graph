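# QA test sound cue for a looping binaural bird song ambience, routed to the Fort_Ambient_Athena sound class.
# A Random node picks among five equally weighted mono bird song waves, then feeds through a Delay and a Looping node into a Mixer whose second input is empty.
# A test attenuation asset named for no air absorption handles distance behavior, and the cue's volume is slightly boosted.
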SoundCue /Game/QA/Art_QA/BR_Audio/Audio/SC_Test_Binaural_Bird_Loop
root=Mixer_0
nodes [
 {id=Delay_0, children [Random_0]}
 {id=Looping_0, children [Delay_0]}
 {id=Mixer_0, children [Looping_0,null], properties {inputVolume [1,1]}}
 {
  id=Random_0
  children [WavePlayer_0,WavePlayer_1,WavePlayer_2,WavePlayer_3,WavePlayer_4]
  properties {bShouldExcludeFromBranchCulling=true, weights [1,1,1,1,1]}
 }
 {id=WavePlayer_0, properties {soundWaveAssetPtr=&/test_binaural_bird_song_mono_01}}
 {id=WavePlayer_1, properties {soundWaveAssetPtr=&/test_binaural_bird_song_mono_02}}
 {id=WavePlayer_2, properties {soundWaveAssetPtr=&/test_binaural_bird_song_mono_03}}
 {id=WavePlayer_3, properties {soundWaveAssetPtr=&/test_binaural_bird_song_mono_04}}
 {id=WavePlayer_4, properties {soundWaveAssetPtr=&/test_binaural_bird_song_mono_05}}
]
properties {
 attenuationSettings=&/test_binarual_attenuation_no_air_absorb
 maxDistance=5750
 soundClassObject=&Fort_Ambient_Athena
 volumeMultiplier=1.25
}
