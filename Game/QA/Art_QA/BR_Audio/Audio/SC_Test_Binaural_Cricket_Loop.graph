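# SC_Test_Binaural_Cricket_Loop is a QA test SoundCue that loops a mono cricket sound, named as a binaural audio test.
# A single looping Wave Player is the root, with no other nodes in the graph.
# A shared test attenuation asset, named for binaural use without air absorption, handles spatialization and falloff, and the cue uses the Fort_Ambient_Athena sound class.
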SoundCue /Game/QA/Art_QA/BR_Audio/Audio/SC_Test_Binaural_Cricket_Loop
root=WavePlayer_2
nodes [{id=WavePlayer_2, properties {bLooping=true, soundWaveAssetPtr=&/test_binaural_cricket_mono_loop}}]
properties {
 attenuationSettings=&/test_binarual_attenuation_no_air_absorb
 maxDistance=5750
 soundClassObject=&Fort_Ambient_Athena
 volumeMultiplier=1
}
