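# QA test Sound Cue that plays a looping mono helicopter sound for testing binaural audio.
# A single looping Wave Player is the root, with no mixing, modulation or randomization.
# The cue uses a test binaural attenuation asset named for having no air absorption, routes to the Fort vehicle engine sound class, and boosts its volume.
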SoundCue /Game/QA/Art_QA/BR_Audio/Audio/SC_Test_Binaural_Helicopter_Loop
root=WavePlayer_5
nodes [{id=WavePlayer_5, properties {bLooping=true, soundWaveAssetPtr=&/test_binaural_helicopter_mono_loop}}]
properties {
 attenuationSettings=&/test_binarual_attenuation_no_air_absorb
 maxDistance=5750
 soundClassObject=&Fort_Vehicles_Engine
 volumeMultiplier=3
}
