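# SC_Test_Binaural_Weapon_Fire_Loop is a QA test Sound Cue for binaural playback of looping assault rifle fire.
# A Looping node repeats a Random node that picks between two close-range mono rifle waves at equal weight and is excluded from branch culling.
# It uses a binaural test attenuation asset named for no air absorption, a ranged weapon projectile effects sound class and boosted volume.
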SoundCue /Game/QA/Art_QA/BR_Audio/Audio/SC_Test_Binaural_Weapon_Fire_Loop
root=Looping_0
nodes [
 {id=Looping_0, children [Random_1]}
 {id=Random_1, children [WavePlayer_0,WavePlayer_1], properties {bShouldExcludeFromBranchCulling=true, weights [1,1]}}
 {id=WavePlayer_0, properties {soundWaveAssetPtr=&/test_binaural_weapon_ar_01_close_mono}}
 {id=WavePlayer_1, properties {soundWaveAssetPtr=&/test_binaural_weapon_ar_02_close_mono}}
]
properties {
 attenuationSettings=&/test_binarual_attenuation_no_air_absorb
 maxDistance=5750
 soundClassObject=&Fort_WeaponRanged_Projectile_Effects
 volumeMultiplier=2
}
