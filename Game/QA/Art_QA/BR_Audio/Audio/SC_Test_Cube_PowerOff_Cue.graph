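# A QA test SoundCue in the BR audio art-test folder, named for a test cube's power-off sound.
# A single wave player plays the Creative toy ActionTrigger power-off sample directly as the root, with no randomization or modulation.
# The cue runs at reduced volume in the Creative gadgets sound class, with small-toy attenuation assigned and the attenuation override enabled.
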SoundCue /Game/QA/Art_QA/BR_Audio/Audio/SC_Test_Cube_PowerOff_Cue
root=WavePlayer_1
nodes [{id=WavePlayer_1, properties {soundWaveAssetPtr=/Game/Sounds/Creative/Toys/ActionTrigger/ActionTrigger_PowerOff_01}}]
properties {
 attenuationSettings=/Game/Sounds/Creative/Attenuations/Fort_Creative_SmallToy_Att
 bOverrideAttenuation=true
 maxDistance=4000
 priority=0.9
 soundClassObject=&Fort_Gadgets_Creative
 volumeMultiplier=0.2
}
