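# SC_Test_Cube_PowerOn_Cue is a QA audio test cue that plays a single Creative toy action-trigger power-on sound, apparently for a test cube.
# A lone wave player feeds the output directly, with no randomization or mixing.
# The cue runs quietly in the Creative gadgets sound class and references small-toy attenuation settings, with attenuation override enabled to set its falloff distance.
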SoundCue /Game/QA/Art_QA/BR_Audio/Audio/SC_Test_Cube_PowerOn_Cue
root=WavePlayer_0
nodes [{id=WavePlayer_0, properties {soundWaveAssetPtr=/Game/Sounds/Creative/Toys/ActionTrigger/ActionTrigger_PowerOn_01}}]
properties {
 attenuationSettings=/Game/Sounds/Creative/Attenuations/Fort_Creative_SmallToy_Att
 bOverrideAttenuation=true
 maxDistance=1400
 soundClassObject=&Fort_Gadgets_Creative
 volumeMultiplier=0.2
 attenuationOverrides {falloffDistance=1000}
}
