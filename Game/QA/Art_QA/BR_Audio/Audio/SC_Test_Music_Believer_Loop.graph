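# Test sound cue in the Art QA BR audio area that loops the Believer lobby music track.
# A single looping wave player feeds the cue, which uses a dedicated attenuation asset and the Fort_Music sound class.
# Base submix and submix sends are disabled, so the audio is routed through a pre-effect send to the SB_Music_Believer_Loop source bus instead of playing directly to the mix.
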SoundCue /Game/QA/Art_QA/BR_Audio/Audio/SC_Test_Music_Believer_Loop
root=WavePlayer_0
nodes [{id=WavePlayer_0, properties {bLooping=true, soundWaveAssetPtr=&/MX_Lobby_Believer}}]
properties {
 attenuationSettings=&/ATT_Music_Believer_Loop
 bEnableBaseSubmix=false
 bEnableSubmixSends=false
 maxDistance=19850
 soundClassObject=/Game/Sounds/FortSoundClassesAndModes/Fort_Music
 volumeMultiplier=1
 preEffectBusSends [{soundSourceBus=&/SB_Music_Believer_Loop}]
}
