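# QA test sound cue that plays a continuous loop of chrome TV tuning noises as an ambient prop sound.
# A looping node repeats a short random delay followed by an evenly weighted random pick from five tuning waves.
# A root modulator wraps the loop for pitch and volume variation.
# It uses a test attenuation asset and the Athena ambient props sound class, with overridden concurrency that prevents new instances at the limit.
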SoundCue /Game/QA/Art_QA/BR_Audio/Audio/SC_Test_TV_Loop
root=Modulator_0
nodes [
 {id=Delay_1, children [Random_0], properties {delayMax=1, delayMin=0.5}}
 {id=Looping_0, children [Delay_1]}
 {id=Modulator_0, children [Looping_0]}
 {id=Random_0, children [WavePlayer_2,WavePlayer_3,WavePlayer_4,WavePlayer_5,WavePlayer_6], properties {weights [1,1,1,1,1]}}
 {id=WavePlayer_2, properties {soundWaveAssetPtr=/Game/Athena/Sounds/Quests/ChromeTV/TV_Chrome_Tuning_1}}
 {id=WavePlayer_3, properties {soundWaveAssetPtr=/Game/Athena/Sounds/Quests/ChromeTV/TV_Chrome_Tuning_2}}
 {id=WavePlayer_4, properties {soundWaveAssetPtr=/Game/Athena/Sounds/Quests/ChromeTV/TV_Chrome_Tuning_3}}
 {id=WavePlayer_5, properties {soundWaveAssetPtr=/Game/Athena/Sounds/Quests/ChromeTV/TV_Chrome_Tuning_4}}
 {id=WavePlayer_6, properties {soundWaveAssetPtr=/Game/Athena/Sounds/Quests/ChromeTV/TV_Chrome_Tuning_5}}
]
properties {
 attenuationSettings=&/ATT_Test_TV_Loop
 bOverrideConcurrency=true
 maxDistance=4200
 soundClassObject=&Fort_Ambient_Props_Athena
 volumeMultiplier=0.55
 attenuationOverrides {lPFRadiusMin=0}
 concurrencyOverrides {resolutionRule=PreventNew}
}
