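# SC_Test_VO_Count_Loop is a QA test SoundCue that endlessly repeats two placeholder voice-over count clips marked do-not-ship.
# A Looping root wraps a short Delay feeding a Concatenator, which plays count 01, then a second short Delay into count 02, so every clip follows a brief gap.
# It uses a test attenuation asset, the engine SFX sound class and a boosted volume.
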SoundCue /Game/QA/Art_QA/BR_Audio/Audio/SC_Test_VO_Count_Loop
root=Looping_0
nodes [
 {id=Concatenator_0, children [WavePlayer_1,Delay_2], properties {inputVolume [1,1]}}
 {id=Delay_1, children [Concatenator_0], properties {delayMax=0.2, delayMin=0.2}}
 {id=Delay_2, children [WavePlayer_2], properties {delayMax=0.2, delayMin=0.2}}
 {id=Looping_0, children [Delay_1]}
 {id=WavePlayer_1, properties {soundWaveAssetPtr=&/VO_Count_01_DO_NOT_SHIP}}
 {id=WavePlayer_2, properties {soundWaveAssetPtr=&/VO_Count_02_DO_NOT_SHIP}}
]
properties {
 attenuationSettings=&/ATT_Test_VO_Count
 maxDistance=4000
 soundClassObject=/Engine/EngineSounds/SFX
 volumeMultiplier=2
}
