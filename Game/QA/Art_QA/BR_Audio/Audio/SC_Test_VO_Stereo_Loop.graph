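# Sound Cue in the Art QA audio folder that plays a stereo voice-over test wave on a loop.
# A single looping Wave Player feeds the output, with boosted volume, the engine SFX sound class and a test VO attenuation asset.
# Its own concurrency override allows one instance at a time, stopping the oldest, and virtualization is disabled.
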SoundCue /Game/QA/Art_QA/BR_Audio/Audio/SC_Test_VO_Stereo_Loop
root=WavePlayer_0
nodes [{id=WavePlayer_0, properties {bLooping=true, soundWaveAssetPtr=&/VO_Stereo_Test}}]
properties {
 attenuationSettings=&/ATT_Test_VO_Count
 bOverrideConcurrency=true
 maxDistance=4000
 soundClassObject=/Engine/EngineSounds/SFX
 virtualizationMode=Disabled
 volumeMultiplier=2
 concurrencyOverrides {maxCount=1, resolutionRule=StopOldest}
}
