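# QA test sound cue in the Fort_Gadgets sound class that plays a looping dump truck idle engine sound.
# A looping wave player feeds an enveloper root whose volume curve fades in from silence over a short ramp.
# Spatialization uses a dedicated test truck attenuation asset, while a concurrency override caps simultaneous voices at two.
# A distance-based submix send feeds the default Fortnite reverb.
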SoundCue /Game/QA/Art_QA/BR_Audio/Audio/SC_Test_Vehicle_Dump_Truck_Idle_Loop
root=Enveloper_1
nodes [
 {
  id=Enveloper_1
  children [WavePlayer_1]
  properties {pitchCurve {editorCurveData {keys [{time=0}]}}, volumeCurve {editorCurveData {keys [<at 0 0>,<at 0.5 1>]}}}
 }
 {id=WavePlayer_1, properties {bLooping=true, soundWaveAssetPtr=/AmbientAudioContent/Waves/SpottedLoops/Spot_Vehicle_Dump_Truck_Idle_Loop}}
]
properties {
 attenuationSettings=&/ATT_Test_Truck_Idle_Loop
 bEnableBusSends=false
 bOverrideConcurrency=true
 maxDistance=9000
 soundClassObject=&Fort_Gadgets
 attenuationOverrides {
  bAttenuateWithLPF=true
  bEnableListenerFocus=true
  distanceAlgorithm=Custom
  falloffDistance=4500
  lPFFrequencyAtMax=750
  lPFRadiusMax=4500
  nonFocusAzimuth=70
  nonFocusVolumeAttenuation=0.5
  attenuationShapeExtents {x=0}
 }
 concurrencyOverrides {maxCount=2, voiceStealReleaseTime=0.2}
 soundSubmixSends [{
   maxSendDistance=4500
   minSendDistance=1000
   minSendLevel=0
   soundSubmix=&Fortnite_Reverb_Default
   customSendLevelCurve {editorCurveData {keys [<at 0 0.4>,<at 1 1>]}}
  }]
}
