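# Test sound submix for the BR_Audio art QA area, named TEST_BR_Audio_Submix.
# It has no parent submix, no child submixes and no non-default properties, so it is a bare standalone submix with default settings.
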SoundSubmix /Game/QA/Art_QA/BR_Audio/Audio/TEST_BR_Audio_Submix
graph_type=sound_submix
parent=null
children []
properties {}
