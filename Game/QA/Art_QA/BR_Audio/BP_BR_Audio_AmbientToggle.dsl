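; QA test actor that toggles the ambient audio console variable when a player pawn overlaps its cube.
; BeginPlay reads au.AmbientAudio.Enabled into a flag and sets a TextRender status label, which notes that disabling requires a restart.
; On overlap, the actor casts to FortPlayerPawnAthena, falling back to FortPlayerPawn, then ExecuteCommand flips the variable through that pawn's PlayerController and updates the flag and label.
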(event OnComponentBeginOverlap(Cube)
  (params (OverlappedComponent :type Primitive%20Component%20Object%20Reference) (OtherActor :type Actor%20Object%20Reference) (OtherComp :type Primitive%20Component%20Object%20Reference) (OtherBodyIndex :type Integer) (bFromSweep :type Boolean) (SweepResult :type Hit%20Result%20Structure%20%28by%20ref%29))
  (bind _asfort_player_pawn_athena (Utilities|Casting|CastToFortPlayerPawnAthena OtherActor)
    (:then
      (CallFunction|ExecuteCommand _asfort_player_pawn_athena))
    (:CastFailed
      (bind _asfort_player_pawn (Utilities|Casting|CastToFortPlayerPawn OtherActor))
      (CallFunction|ExecuteCommand _asfort_player_pawn))))

(event EventBeginPlay
  (bind _ambiencetoggle (Variables|Default|GetAmbienceToggle))
  (bind _returnvalue (Development|GetConsoleVariableIntValue "au.AmbientAudio.Enabled"))
  (bind _output_get (Variables|Default|SetGetAmbientStatus _returnvalue))
  (Variables|Default|SetIsAmbienceEnabled (== _output_get 1))
  (if (Variables|Default|GetIsAmbienceEnabled)
    (Rendering|Components|TextRender|SetText _ambiencetoggle (Utilities|Text|ToText(String) "Ambient Audio is ENABLED"))
    (else
      (Rendering|Components|TextRender|SetText _ambiencetoggle (Utilities|Text|ToText(String) "Ambient Audio is DISABLED (Requires Restart to take effect)")))))

(event EventTick
  (params (DeltaSeconds :type Float%20%28single-precision%29)))

(fn ExecuteCommand
  (params (PlayerPawn :type Pawn%20Object%20Reference))
  (bind _asplayer_controller (Utilities|Casting|CastToPlayerController (Pawn|GetController PlayerPawn)))
  (if (Variables|Default|GetIsAmbienceEnabled)
    (Development|ExecuteConsoleCommand "au.AmbientAudio.Enabled 0" _asplayer_controller)
    (Variables|Default|SetIsAmbienceEnabled false)
    (Rendering|Components|TextRender|SetText (Variables|Default|GetAmbienceToggle) (Utilities|Text|ToText(String) "Ambient Audio is DISABLED (Requires Restart to take effect)"))
    (else
      (Development|ExecuteConsoleCommand "au.AmbientAudio.Enabled 1" _asplayer_controller)
      (Variables|Default|SetIsAmbienceEnabled true)
      (Rendering|Components|TextRender|SetText (Variables|Default|GetAmbienceToggle) (Utilities|Text|ToText(String) "Ambient Audio is ENABLED")))))

(fn ConstructionScript)
