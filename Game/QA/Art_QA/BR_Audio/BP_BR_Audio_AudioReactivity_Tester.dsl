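; QA test actor Blueprint for audio reactivity, which uses a trigger box to switch test audio and analysis debugging on and off.
; Beginning an overlap with TestBox activates the Audio component and sets the AudioAnalysis.DebugEnabled console variable to 1.
; Ending the overlap deactivates the component and sets the variable back to 0.
; Neither event filters the overlapping actor, and the construction script is empty.
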(event OnComponentBeginOverlap(TestBox)
  (params (OverlappedComponent :type Primitive%20Component%20Object%20Reference) (OtherActor :type Actor%20Object%20Reference) (OtherComp :type Primitive%20Component%20Object%20Reference) (OtherBodyIndex :type Integer) (bFromSweep :type Boolean) (SweepResult :type Hit%20Result%20Structure%20%28by%20ref%29))
  (Components|Activation|Activate (Variables|Default|GetAudio) false)
  (Development|ExecuteConsoleCommand "AudioAnalysis.DebugEnabled 1"))

(event OnComponentEndOverlap(TestBox)
  (params (OverlappedComponent :type Primitive%20Component%20Object%20Reference) (OtherActor :type Actor%20Object%20Reference) (OtherComp :type Primitive%20Component%20Object%20Reference) (OtherBodyIndex :type Integer))
  (Components|Activation|Deactivate (Variables|Default|GetAudio))
  (Development|ExecuteConsoleCommand "AudioAnalysis.DebugEnabled 0"))

(fn ConstructionScript)
