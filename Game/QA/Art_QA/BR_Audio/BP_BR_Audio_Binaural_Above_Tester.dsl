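; QA audio test actor, named as a binaural-above tester, that switches an audio component on and off through a cube trigger.
; BeginPlay gives the cube its purple idle material.
; Each Athena player pawn overlap of the collision volume flips isMusicPlaying, toggling the audio, swapping the cube between purple and TV-screen materials, and playing a power-on or power-off cue at the cube.
; The build-zone arrow component also changes activation and visibility.
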(event EventBeginPlay
  (Rendering|Material|SetMaterial (Variables|Default|GetCube) 0 "/Game/Athena/Prototype/Blueprints/Cube/Materials/M_Cube_Purple_P5.M_Cube_Purple_P5"))

(event OnComponentBeginOverlap(Collision_Volume)
  (params (OverlappedComponent :type Primitive%20Component%20Object%20Reference) (OtherActor :type Actor%20Object%20Reference) (OtherComp :type Primitive%20Component%20Object%20Reference) (OtherBodyIndex :type Integer) (bFromSweep :type Boolean) (SweepResult :type Hit%20Result%20Structure%20%28by%20ref%29))
  (Utilities|Casting|CastToFortPlayerPawnAthena OtherActor)
  (if (Variables|Default|GetisMusicPlaying)
    (Components|Activation|Deactivate (Variables|Default|GetAudio))
    (Variables|Default|SetisMusicPlaying false)
    (Rendering|Material|SetMaterial (Variables|Default|GetCube) 0 "/Game/Athena/Prototype/Blueprints/Cube/Materials/M_Cube_Purple_P5.M_Cube_Purple_P5")
    (Audio|PlaySoundatLocation "/Game/QA/Art_QA/BR_Audio/Audio/SC_Test_Cube_PowerOff_Cue.SC_Test_Cube_PowerOff_Cue" (Transformation|GetWorldLocation (Variables|Default|GetCube)))
    (Components|Activation|Deactivate (Variables|Default|GetBR_Audio_Mesh_BuildZone_Arrow01))
    (else
      (Variables|Default|SetisMusicPlaying true)
      (Components|Activation|Activate (Variables|Default|GetAudio) false)
      (Rendering|Material|SetMaterial (Variables|Default|GetCube) 0 "/Game/Environments/Helios/Props/Appliance/Appliance_RetroConsole_A/Materials/MI_Appliance_RetroTVScreen_A.MI_Appliance_RetroTVScreen_A")
      (Audio|PlaySoundatLocation "/Game/QA/Art_QA/BR_Audio/Audio/SC_Test_Cube_PowerOn_Cue.SC_Test_Cube_PowerOn_Cue" (Transformation|GetWorldLocation (Variables|Default|GetCube)))
      (Components|Activation|Activate (Variables|Default|GetBR_Audio_Mesh_BuildZone_Arrow01) false)))
  (Rendering|ToggleVisibility (Variables|Default|GetBR_Audio_Mesh_BuildZone_Arrow01) false))

(fn ConstructionScript)
