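; Audio QA test actor that exercises sound concurrency resolution rules in a Fortnite test map.
; Seven box triggers, entered by an Athena player pawn, each pick one SoundConcurrencySet entry, from Prevent New to Stop Farthest Then Prevent New.
; Each spawns two or three delayed stinger cues on close or far attach points.
; Priority, distance or volume varies where the rule under test depends on it.
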(event OnComponentBeginOverlap(BoxCollide_C1)
  (params (OverlappedComponent :type Primitive%20Component%20Object%20Reference) (OtherActor :type Actor%20Object%20Reference) (OtherComp :type Primitive%20Component%20Object%20Reference) (OtherBodyIndex :type Integer) (bFromSweep :type Boolean) (SweepResult :type Hit%20Result%20Structure%20%28by%20ref%29))
  ;[0 0 0] Prevent New Test
  (Utilities|Casting|CastToFortPlayerPawnAthena OtherActor)
  (bind _result (ArrayMacros|SafeArrayGet (Variables|Default|GetSoundConcurrencySet))
    (:%20%20
      (bind _returnvalue (Audio|SpawnSoundAttached :Sound "/Game/Sounds/Fort_Test_Sounds/Test_Map_Sounds/DadBroStingerTest_HighPri_Cue.DadBroStingerTest_HighPri_Cue" :AttachToComponent (Variables|Default|GetTestSoundClose) :AttachPointName "None" :LocationType "KeepRelativeOffset" :bStopWhenAttachedToDestroyed true :VolumeMultiplier 1.0 :PitchMultiplier 1.0 :StartTime 0.0 :ConcurrencySettings _result :bAutoDestroy true))
      (Components|Activation|Activate _returnvalue false)
      (Utilities|FlowControl|Delay 1.0)
      (bind _result_2 (ArrayMacros|SafeArrayGet (Variables|Default|GetSoundConcurrencySet))
        (:%20%20
          (bind _returnvalue_2 (Audio|SpawnSoundAttached :Sound "/Game/Sounds/Fort_Test_Sounds/Test_Map_Sounds/DadBroStingerTest_HighPri_Cue.DadBroStingerTest_HighPri_Cue" :AttachToComponent (Variables|Default|GetTestSoundClose) :AttachPointName "None" :LocationType "KeepRelativeOffset" :bStopWhenAttachedToDestroyed true :VolumeMultiplier 1.0 :PitchMultiplier 1.0 :StartTime 0.0 :ConcurrencySettings _result_2 :bAutoDestroy true))
          (Components|Activation|Activate _returnvalue_2 false))) ; 0 is Prevent New
    )) ; 0 is Prevent New
)

(event OnComponentBeginOverlap(BoxCollide_C2)
  (params (OverlappedComponent :type Primitive%20Component%20Object%20Reference) (OtherActor :type Actor%20Object%20Reference) (OtherComp :type Primitive%20Component%20Object%20Reference) (OtherBodyIndex :type Integer) (bFromSweep :type Boolean) (SweepResult :type Hit%20Result%20Structure%20%28by%20ref%29))
  ;[0 0 0] Stop Oldest Test
  (Utilities|Casting|CastToFortPlayerPawnAthena OtherActor)
  (bind _result (ArrayMacros|SafeArrayGet (Variables|Default|GetSoundConcurrencySet) 1)
    (:%20%20
      (bind _returnvalue (Audio|SpawnSoundAttached :Sound "/Game/Sounds/Fort_Test_Sounds/Test_Map_Sounds/DadBroStingerTest_HighPri_Cue.DadBroStingerTest_HighPri_Cue" :AttachToComponent (Variables|Default|GetTestSoundClose) :AttachPointName "None" :LocationType "KeepRelativeOffset" :bStopWhenAttachedToDestroyed true :VolumeMultiplier 1.0 :PitchMultiplier 1.0 :StartTime 0.0 :ConcurrencySettings _result :bAutoDestroy true))
      (Components|Activation|Activate _returnvalue false)
      (Utilities|FlowControl|Delay 1.0)
      (bind _result_2 (ArrayMacros|SafeArrayGet (Variables|Default|GetSoundConcurrencySet) 1)
        (:%20%20
          (bind _returnvalue_2 (Audio|SpawnSoundAttached :Sound "/Game/Sounds/Fort_Test_Sounds/Test_Map_Sounds/DadBroStingerTest_HighPri_Cue.DadBroStingerTest_HighPri_Cue" :AttachToComponent (Variables|Default|GetTestSoundClose) :AttachPointName "None" :LocationType "KeepRelativeOffset" :bStopWhenAttachedToDestroyed true :VolumeMultiplier 1.0 :PitchMultiplier 1.0 :StartTime 0.0 :ConcurrencySettings _result_2 :bAutoDestroy true))
          (Components|Activation|Activate _returnvalue_2 false))) ; 1 is Stop Oldest
    )) ; 1 is Stop Oldest
)

(event OnComponentBeginOverlap(BoxCollide_C3)
  (params (OverlappedComponent :type Primitive%20Component%20Object%20Reference) (OtherActor :type Actor%20Object%20Reference) (OtherComp :type Primitive%20Component%20Object%20Reference) (OtherBodyIndex :type Integer) (bFromSweep :type Boolean) (SweepResult :type Hit%20Result%20Structure%20%28by%20ref%29))
  ;[0 0 0] Stop Quietest Test
  (Utilities|Casting|CastToFortPlayerPawnAthena OtherActor)
  (bind _result (ArrayMacros|SafeArrayGet (Variables|Default|GetSoundConcurrencySet) 2)
    (:%20%20
      (bind _returnvalue (Audio|SpawnSoundAttached :Sound "/Game/Sounds/Fort_Test_Sounds/Test_Map_Sounds/DadBroStingerTest_HighPri_Cue.DadBroStingerTest_HighPri_Cue" :AttachToComponent (Variables|Default|GetTestSoundFar) :AttachPointName "None" :LocationType "KeepRelativeOffset" :bStopWhenAttachedToDestroyed true :VolumeMultiplier 0.5 :PitchMultiplier 1.0 :StartTime 0.0 :ConcurrencySettings _result :bAutoDestroy true))
      (Components|Activation|Activate _returnvalue false)
      (Utilities|FlowControl|Delay 1.0)
      (bind _result_2 (ArrayMacros|SafeArrayGet (Variables|Default|GetSoundConcurrencySet) 2)
        (:%20%20
          (bind _returnvalue_2 (Audio|SpawnSoundAttached :Sound "/Game/Sounds/Fort_Test_Sounds/Test_Map_Sounds/DadBroStingerTest_HighPri_Cue.DadBroStingerTest_HighPri_Cue" :AttachToComponent (Variables|Default|GetTestSoundClose) :AttachPointName "None" :LocationType "KeepRelativeOffset" :bStopWhenAttachedToDestroyed true :VolumeMultiplier 1.0 :PitchMultiplier 1.0 :StartTime 0.0 :ConcurrencySettings _result_2 :bAutoDestroy true))
          (Components|Activation|Activate _returnvalue_2 false))) ; 2 is Stop Quietest
    )) ; 2 is Stop Quietest
)

(event OnComponentBeginOverlap(BoxCollide_C4)
  (params (OverlappedComponent :type Primitive%20Component%20Object%20Reference) (OtherActor :type Actor%20Object%20Reference) (OtherComp :type Primitive%20Component%20Object%20Reference) (OtherBodyIndex :type Integer) (bFromSweep :type Boolean) (SweepResult :type Hit%20Result%20Structure%20%28by%20ref%29))
  ;[0 0 0] Stop Lowest Priority Test
  (Utilities|Casting|CastToFortPlayerPawnAthena OtherActor)
  (bind _result (ArrayMacros|SafeArrayGet (Variables|Default|GetSoundConcurrencySet) 3)
    (:%20%20
      (bind _returnvalue (Audio|SpawnSoundAttached :Sound "/Game/Sounds/Fort_Test_Sounds/Test_Map_Sounds/DadBroStingerTest_LowPri_Cue.DadBroStingerTest_LowPri_Cue" :AttachToComponent (Variables|Default|GetTestSoundClose) :AttachPointName "None" :LocationType "KeepRelativeOffset" :bStopWhenAttachedToDestroyed true :VolumeMultiplier 1.0 :PitchMultiplier 1.0 :StartTime 0.0 :ConcurrencySettings _result :bAutoDestroy true))
      (Components|Activation|Activate _returnvalue false)
      (Utilities|FlowControl|Delay 1.0)
      (bind _result_2 (ArrayMacros|SafeArrayGet (Variables|Default|GetSoundConcurrencySet) 3)
        (:%20%20
          (bind _returnvalue_2 (Audio|SpawnSoundAttached :Sound "/Game/Sounds/Fort_Test_Sounds/Test_Map_Sounds/DadBroStingerTest_HighPri_Cue.DadBroStingerTest_HighPri_Cue" :AttachToComponent (Variables|Default|GetTestSoundClose) :AttachPointName "None" :LocationType "KeepRelativeOffset" :bStopWhenAttachedToDestroyed true :VolumeMultiplier 1.0 :PitchMultiplier 1.0 :StartTime 0.0 :ConcurrencySettings _result_2 :bAutoDestroy true))
          (Components|Activation|Activate _returnvalue_2 false))) ; 3 is Stop Lowest Priority
    )) ; 3 is Stop Lowest Priority
)

(event OnComponentBeginOverlap(BoxCollide_C5)
  (params (OverlappedComponent :type Primitive%20Component%20Object%20Reference) (OtherActor :type Actor%20Object%20Reference) (OtherComp :type Primitive%20Component%20Object%20Reference) (OtherBodyIndex :type Integer) (bFromSweep :type Boolean) (SweepResult :type Hit%20Result%20Structure%20%28by%20ref%29))
  ;[0 0 0] Stop Lowest Priority Then Prevent New Test
  (Utilities|Casting|CastToFortPlayerPawnAthena OtherActor)
  (bind _result (ArrayMacros|SafeArrayGet (Variables|Default|GetSoundConcurrencySet) 4)
    (:%20%20
      (bind _returnvalue (Audio|SpawnSoundAttached :Sound "/Game/Sounds/Fort_Test_Sounds/Test_Map_Sounds/DadBroStingerTest_LowPri_Cue.DadBroStingerTest_LowPri_Cue" :AttachToComponent (Variables|Default|GetTestSoundClose) :AttachPointName "None" :LocationType "KeepRelativeOffset" :bStopWhenAttachedToDestroyed true :VolumeMultiplier 1.0 :PitchMultiplier 1.0 :StartTime 0.0 :ConcurrencySettings _result :bAutoDestroy true))
      (Components|Activation|Activate _returnvalue false)
      (Utilities|FlowControl|Delay 1.0)
      (bind _result_2 (ArrayMacros|SafeArrayGet (Variables|Default|GetSoundConcurrencySet) 4)
        (:%20%20
          (bind _returnvalue_2 (Audio|SpawnSoundAttached :Sound "/Game/Sounds/Fort_Test_Sounds/Test_Map_Sounds/DadBroStingerTest_LowPri_Cue.DadBroStingerTest_LowPri_Cue" :AttachToComponent (Variables|Default|GetTestSoundClose) :AttachPointName "None" :LocationType "KeepRelativeOffset" :bStopWhenAttachedToDestroyed true :VolumeMultiplier 1.0 :PitchMultiplier 1.0 :StartTime 0.0 :ConcurrencySettings _result_2 :bAutoDestroy true))
          (Components|Activation|Activate _returnvalue_2 false)
          (Utilities|FlowControl|Delay 3.0)
          (bind _result_3 (ArrayMacros|SafeArrayGet (Variables|Default|GetSoundConcurrencySet) 4)
            (:%20%20
              (bind _returnvalue_3 (Audio|SpawnSoundAttached :Sound "/Game/Sounds/Fort_Test_Sounds/Test_Map_Sounds/DadBroStingerTest_HighPri_Cue.DadBroStingerTest_HighPri_Cue" :AttachToComponent (Variables|Default|GetTestSoundFar) :AttachPointName "None" :LocationType "KeepRelativeOffset" :bStopWhenAttachedToDestroyed true :VolumeMultiplier 1.0 :PitchMultiplier 1.0 :StartTime 0.0 :ConcurrencySettings _result_3 :bAutoDestroy true))
              (Components|Activation|Activate _returnvalue_3 false))) ; 4 is Stop Lowest Priority Then Prevent New
        )) ; 4 is Stop Lowest Priority Then Prevent New
    )) ; 4 is Stop Lowest Priority Then Prevent New
)

(event OnComponentBeginOverlap(BoxCollide_C6)
  (params (OverlappedComponent :type Primitive%20Component%20Object%20Reference) (OtherActor :type Actor%20Object%20Reference) (OtherComp :type Primitive%20Component%20Object%20Reference) (OtherBodyIndex :type Integer) (bFromSweep :type Boolean) (SweepResult :type Hit%20Result%20Structure%20%28by%20ref%29))
  ;[0 0 0] Stop Farthest Then Oldest Test
  (Utilities|Casting|CastToFortPlayerPawnAthena OtherActor)
  (bind _result (ArrayMacros|SafeArrayGet (Variables|Default|GetSoundConcurrencySet) 5)
    (:%20%20
      (bind _returnvalue (Audio|SpawnSoundAttached :Sound "/Game/Sounds/Fort_Test_Sounds/Test_Map_Sounds/DadBroStingerTest_HighPri_Cue.DadBroStingerTest_HighPri_Cue" :AttachToComponent (Variables|Default|GetTestSoundFar) :AttachPointName "None" :LocationType "KeepRelativeOffset" :bStopWhenAttachedToDestroyed true :VolumeMultiplier 1.0 :PitchMultiplier 1.0 :StartTime 0.0 :ConcurrencySettings _result :bAutoDestroy true))
      (Components|Activation|Activate _returnvalue false)
      (Utilities|FlowControl|Delay 1.0)
      (bind _result_2 (ArrayMacros|SafeArrayGet (Variables|Default|GetSoundConcurrencySet) 5)
        (:%20%20
          (bind _returnvalue_2 (Audio|SpawnSoundAttached :Sound "/Game/Sounds/Fort_Test_Sounds/Test_Map_Sounds/DadBroStingerTest_HighPri_Cue.DadBroStingerTest_HighPri_Cue" :AttachToComponent (Variables|Default|GetTestSoundClose) :AttachPointName "None" :LocationType "KeepRelativeOffset" :bStopWhenAttachedToDestroyed true :VolumeMultiplier 1.0 :PitchMultiplier 1.0 :StartTime 0.0 :ConcurrencySettings _result_2 :bAutoDestroy true))
          (Components|Activation|Activate _returnvalue_2 false)
          (Utilities|FlowControl|Delay 2.0)
          (bind _result_3 (ArrayMacros|SafeArrayGet (Variables|Default|GetSoundConcurrencySet) 5)
            (:%20%20
              (bind _returnvalue_3 (Audio|SpawnSoundAttached :Sound "/Game/Sounds/Fort_Test_Sounds/Test_Map_Sounds/DadBroStingerTest_HighPri_Cue.DadBroStingerTest_HighPri_Cue" :AttachToComponent (Variables|Default|GetTestSoundClose) :AttachPointName "None" :LocationType "KeepRelativeOffset" :bStopWhenAttachedToDestroyed true :VolumeMultiplier 1.0 :PitchMultiplier 1.0 :StartTime 0.0 :ConcurrencySettings _result_3 :bAutoDestroy true))
              (Components|Activation|Activate _returnvalue_3 false))) ; 5 is Stop Farthest Then Oldest
        )) ; 5 is Stop Farthest Then Oldest
    )) ; 5 is Stop Farthest Then Oldest
)

(event OnComponentBeginOverlap(BoxCollide_C7)
  (params (OverlappedComponent :type Primitive%20Component%20Object%20Reference) (OtherActor :type Actor%20Object%20Reference) (OtherComp :type Primitive%20Component%20Object%20Reference) (OtherBodyIndex :type Integer) (bFromSweep :type Boolean) (SweepResult :type Hit%20Result%20Structure%20%28by%20ref%29))
  ;[0 0 0] Stop Farthest Then Prevent New Test
  (Utilities|Casting|CastToFortPlayerPawnAthena OtherActor)
  (bind _result (ArrayMacros|SafeArrayGet (Variables|Default|GetSoundConcurrencySet) 6)
    (:%20%20
      (bind _returnvalue (Audio|SpawnSoundAttached :Sound "/Game/Sounds/Fort_Test_Sounds/Test_Map_Sounds/DadBroStingerTest_HighPri_Cue.DadBroStingerTest_HighPri_Cue" :AttachToComponent (Variables|Default|GetTestSoundFar) :AttachPointName "None" :LocationType "KeepRelativeOffset" :bStopWhenAttachedToDestroyed true :VolumeMultiplier 1.0 :PitchMultiplier 1.0 :StartTime 0.0 :ConcurrencySettings _result :bAutoDestroy true))
      (Components|Activation|Activate _returnvalue false)
      (Utilities|FlowControl|Delay 1.0)
      (bind _result_2 (ArrayMacros|SafeArrayGet (Variables|Default|GetSoundConcurrencySet) 6)
        (:%20%20
          (bind _returnvalue_2 (Audio|SpawnSoundAttached :Sound "/Game/Sounds/Fort_Test_Sounds/Test_Map_Sounds/DadBroStingerTest_HighPri_Cue.DadBroStingerTest_HighPri_Cue" :AttachToComponent (Variables|Default|GetTestSoundClose) :AttachPointName "None" :LocationType "KeepRelativeOffset" :bStopWhenAttachedToDestroyed true :VolumeMultiplier 1.0 :PitchMultiplier 1.0 :StartTime 0.0 :ConcurrencySettings _result_2 :bAutoDestroy true))
          (Components|Activation|Activate _returnvalue_2 false)
          (Utilities|FlowControl|Delay 1.0)
          (bind _result_3 (ArrayMacros|SafeArrayGet (Variables|Default|GetSoundConcurrencySet) 6)
            (:%20%20
              (bind _returnvalue_3 (Audio|SpawnSoundAttached :Sound "/Game/Sounds/Fort_Test_Sounds/Test_Map_Sounds/DadBroStingerTest_HighPri_Cue.DadBroStingerTest_HighPri_Cue" :AttachToComponent (Variables|Default|GetTestSoundClose) :AttachPointName "None" :LocationType "KeepRelativeOffset" :bStopWhenAttachedToDestroyed true :VolumeMultiplier 1.0 :PitchMultiplier 1.0 :StartTime 0.0 :ConcurrencySettings _result_3 :bAutoDestroy true))
              (Components|Activation|Activate _returnvalue_3 false))) ; 6 is Stop Farthest Then Prevent New
        )) ; 6 is Stop Farthest Then Prevent New
    )) ; 6 is Stop Farthest Then Prevent New
)

(fn ConstructionScript)
