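; QA audio test actor that simulates third-person proximity chat at its location, labeled with its sound's name.
; After a random delay, PlayAudio spawns the sound only if no audio component exists, then schedules StopAudio.
; Randomized timers alternate StopAudio, which fires the "StopFiring" trigger, with PlayAudio, which leaves an existing component untouched.
; StopProximityChat destroys the component without clearing timers; DeactivateProximityAudio clears them and fires the trigger, while EndPlay only clears them.
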(event EventBeginPlay
  (bind _3pproximitychatsound (Variables|Default|Get3pProximityChatSound)
    (:then
      (Rendering|Components|TextRender|SetText (Variables|Default|GetTextRender) (Utilities|Text|ToText(String) (Utilities|GetObjectName _3pproximitychatsound))))
    (:else)))

(event Custom|StartProximityChat
  (Utilities|FlowControl|Delay (Math|Random|RandomFloatinRange 0.3 5.0))
  (CallFunction|PlayAudio))

(event Custom|StopProximityChat
  (bind _3paudiocomponent (Variables|Default|Get3pAudioComponent))
  (Audio|Parameter|ExecuteTriggerParameter _3paudiocomponent "StopFiring")
  (Audio|Components|Audio|Stop _3paudiocomponent)
  (Components|DestroyComponent _3paudiocomponent))

(event Custom|PlayAudio
  (Variables|Default|Get3pAudioComponent
    (:then
      _)
    (:else
      (bind _3pproximitychatsound (Variables|Default|Get3pProximityChatSound)
        (:then
          (bind _returnvalue (Audio|SpawnSoundatLocation _3pproximitychatsound (Transformation|GetActorLocation)))
          (Variables|Default|Set3pAudioComponent _returnvalue))
        (:else))))
  (CallFunction|InvalidateProxTimers)
  (bind _returnvalue_2 (Utilities|Time|SetTimerbyEvent (delegate Custom|StartStop) (Math|Random|RandomFloatinRange 1.0 5.0)))
  (Variables|Default|SetStopTimer _returnvalue_2))

(event Custom|StartStop
  (CallFunction|StopAudio))

(event Custom|StopAudio
  (Audio|Parameter|ExecuteTriggerParameter (Variables|Default|Get3pAudioComponent) "StopFiring")
  (CallFunction|InvalidateProxTimers)
  (bind _returnvalue (Utilities|Time|SetTimerbyEvent (delegate Custom|StartPlay) (Math|Random|RandomFloatinRange 0.4 5.0)))
  (Variables|Default|SetPlayTimer _returnvalue))

(event Custom|StartPlay
  (CallFunction|PlayAudio))

(event Custom|InvalidateProxTimers
  (bind _stoptimer (Variables|Default|GetStopTimer))
  (bind _playtimer (Variables|Default|GetPlayTimer))
  (if (Utilities|Time|IsValidTimerHandle _stoptimer)
    (Utilities|Time|ClearandInvalidateTimerbyHandle _stoptimer)
    (else
      _))
  (if (Utilities|Time|IsValidTimerHandle _playtimer)
    (Utilities|Time|ClearandInvalidateTimerbyHandle _playtimer)))

(event EventEndPlay
  (params (EndPlayReason :type EEndPlayReason%20Enum :default "Destroyed"))
  (CallFunction|InvalidateProxTimers))

(event Custom|DeactivateProximityAudio
  (CallFunction|InvalidateProxTimers)
  (bind _3paudiocomponent (Variables|Default|Get3pAudioComponent)
    (:then
      (Audio|Parameter|ExecuteTriggerParameter _3paudiocomponent "StopFiring"))
    (:else)))

(fn ConstructionScript)
