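; QA audio test actor that spams a looping 1 kHz saw test tone as a non-spatialized 2D sound.
; PlayAudio starts a repeating sub-second timer, and each tick spawns a new quiet instance while bActive is true.
; InvalidateTimers clears the timer and destroys the actor.
; Despite the name, nothing is randomized, and spawned sounds are neither stored nor explicitly stopped.
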(event EventPlayAudio
  (bind _returnvalue (Utilities|Time|SetTimerbyEvent (delegate Custom|SpamSound) 0.3 true))
  (Variables|Default|SetSpamTimer _returnvalue))

(event Custom|SpamSound
  (if (Variables|Default|GetbActive)
    (Audio|SpawnSound2D "/Game/Sounds/Fort_Test_Sounds/Tones/1000hz_Saw_Tone_Loop.1000hz_Saw_Tone_Loop" 0.2)))

(event EventInvalidateTimers
  (Utilities|Time|ClearandInvalidateTimerbyHandle (Variables|Default|GetSpamTimer))
  (Actor|DestroyActor))

(fn ConstructionScript
  (|CallParentFunction))
