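; QA audio test actor that simulates intermittent third-person weapon fire and labels itself with the weapon sound's name.
; StartFiring waits a random delay, then PlayAudio spawns the sound at the actor if needed, triggers "StartFiring" and sets a random Character.UniqueID.
; Randomized timers alternate PlayAudio and StopAudio on the same audio component.
; DeactivateAudio clears timers and triggers "StopFiring", whereas the StopFiring event only sends that trigger, leaving timers running.
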(event EventBeginPlay
  (bind _3pweaponsound (Variables|Default|Get3pWeaponSound)
    (:then
      (Rendering|Components|TextRender|SetText (Variables|Default|GetTextRender) (Utilities|Text|ToText(String) (Utilities|GetObjectName _3pweaponsound))))
    (:else)))

(event Custom|StartFiring
  (Utilities|FlowControl|Delay (Math|Random|RandomFloatinRange 0.3 5.0))
  (CallFunction|PlayAudio))

(event Custom|StopFiring
  (Audio|Parameter|ExecuteTriggerParameter (Variables|Default|Get3pAudioComponent) "StopFiring"))

(event Custom|PlayAudio
  (Variables|Default|Get3pAudioComponent
    (:then
      _)
    (:else
      (bind _3pweaponsound (Variables|Default|Get3pWeaponSound)
        (:then
          (bind _returnvalue (Audio|SpawnSoundatLocation _3pweaponsound (Transformation|GetActorLocation)))
          (bind _output_get (Variables|Default|Set3pAudioComponent _returnvalue)))
        (:else))))
  (Audio|Parameter|ExecuteTriggerParameter (Variables|Default|Get3pAudioComponent) "StartFiring")
  (Audio|Parameter|SetFloatParameter _output_get "Character.UniqueID" (Math|Conversions|ToFloat(Integer) (Math|Random|RandomInteger 1000)) :!class /Script/Engine.AudioComponent)
  (CallFunction|InvalidateTimers)
  (bind _returnvalue_2 (Utilities|Time|SetTimerbyEvent (delegate Custom|StartStop) (Math|Random|RandomFloatinRange 1.0 5.0)))
  (Variables|Default|SetStopTimer _returnvalue_2))

(event Custom|StartStop
  (CallFunction|StopAudio))

(event Custom|StopAudio
  (Audio|Parameter|ExecuteTriggerParameter (Variables|Default|Get3pAudioComponent) "StopFiring")
  (CallFunction|InvalidateTimers)
  (bind _returnvalue (Utilities|Time|SetTimerbyEvent (delegate Custom|StartPlay) (Math|Random|RandomFloatinRange 0.4 5.0)))
  (Variables|Default|SetPlayTimer _returnvalue))

(event Custom|StartPlay
  (CallFunction|PlayAudio))

(event Custom|InvalidateTimers
  (bind _stoptimer (Variables|Default|GetStopTimer))
  (bind _playtimer (Variables|Default|GetPlayTimer))
  (if (Utilities|Time|IsValidTimerHandle _stoptimer)
    (Utilities|Time|ClearandInvalidateTimerbyHandle _stoptimer)
    (else
      _))
  (if (Utilities|Time|IsValidTimerHandle _playtimer)
    (Utilities|Time|ClearandInvalidateTimerbyHandle _playtimer)))

(event EventEndPlay
  (params (EndPlayReason :type EEndPlayReason%20Enum :default "Destroyed"))
  (CallFunction|InvalidateTimers))

(event Custom|DeactivateAudio
  (CallFunction|InvalidateTimers)
  (bind _3paudiocomponent (Variables|Default|Get3pAudioComponent)
    (:then
      (Audio|Parameter|ExecuteTriggerParameter _3paudiocomponent "StopFiring"))
    (:else)))

(fn ConstructionScript)
