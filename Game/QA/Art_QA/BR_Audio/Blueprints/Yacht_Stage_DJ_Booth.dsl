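; Destructible yacht stage DJ booth that loops fallback audio, swaps to streamed media with subtitles on cue, and plays a stop sound when destroyed.
; BP_DJBoothMeshTrigger events start and stop playback, with late joiners deriving start offset from server time.
; Non-dedicated instances request the stream URL and open the media player, retrying failures before reverting to fallback audio.
; Clients also toggle picture-in-picture livestream runtime options around playback.
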(event EventOnDeathPlayEffects
  (params (Damage :type Float%20%28single-precision%29) (DamageTags :type Gameplay%20Tag%20Container%20Structure%20%28by%20ref%29) (Momentum :type Vector :default "0, 0, 0") (HitInfo :type Hit%20Result%20Structure%20%28by%20ref%29) (InstigatedBy :type Fort%20Pawn%20Object%20Reference) (DamageCauser :type Actor%20Object%20Reference) (EffectContext :type Gameplay%20Effect%20Context%20Handle%20Structure))
  (Audio|Components|Audio|FadeOut (Variables|Default|GetAudio) 0.5 0.0 "Linear" :!class /Script/Engine.AudioComponent)
  (Variables|Default|SetIsDestroyed true)
  (CallFunction|DjbSoff)
  (Audio|PlaySoundatLocation "/Game/Sounds/Apollo/Props/DonutDJBooth/PGO_Stage_DJ_Booth_MusicStop_Cue.PGO_Stage_DJ_Booth_MusicStop_Cue" (Transformation|GetWorldLocation (Variables|Default|GetAudio))))

(event Custom|djb_son
  (params (AudioUID :type String))
  (if (not (Networking|IsDedicatedServer))
    (CallFunction|StopPIP)
    (Audio|Components|Audio|Stop (Variables|Default|GetAudio))
    (CallFunction|RequestStreamURL)))

(event Custom|djb_soff
  (if (not (Networking|IsDedicatedServer))
    (CallFunction|StopPIP)
    (FortniteSubtitles|SubtitlesPlayer|Stop (Variables|Default|GetSubtitlePlayer))
    (Media|MediaPlayer|Close (Variables|Default|GetMediaPlayer) :!class /Script/MediaAssets.MediaPlayer)
    (CallFunction|UnBindAllEvents)
    (if (not (Variables|Default|GetIsDestroyed))
      (Audio|Components|Audio|Play (Variables|Default|GetAudio) 0.0))))

(event Custom|djb_fon
  (if (not (Networking|IsDedicatedServer))
    (CallFunction|LoadFallback)))

(event EventBeginPlay
  (CallFunction|DJBSetupMeshTrigger)
  (Audio|Components|Audio|Play (Variables|Default|GetAudio) 0.0)
  (if (not (Networking|IsDedicatedServer))
    (CallFunction|StartPip)))

(event Custom|djb_foff
  (if (not (Networking|IsDedicatedServer))
    (Audio|Components|Audio|Stop (Variables|Default|GetAudio))))

(event EventEndPlay
  (params (EndPlayReason :type EEndPlayReason%20Enum :default "Destroyed"))
  (CallFunction|DjbSoff)
  (CallFunction|DjbFoff))

(event Custom|RequestStreamURL
  (bind _mediasource (Variables|Default|GetMediaSource))
  (Variables|/Script/FortniteGame.FortStreamMediaSource|SetStreamID (Variables|Default|GetAudioUID) _mediasource)
  (Stream|BindEventtoOnVideoUrlSuccess _mediasource (EventDispatchers|CreateEvent :!event Custom|StreamingUrlSucceeded))
  (Stream|BindEventtoOnVideoUrlFailed _mediasource (EventDispatchers|CreateEvent :!event Custom|StreamingUrlFailed))
  (Stream|RequestVideoUrl _mediasource (Fortnite|GetFirstLocalFortPlayerController true)))

(event Custom|StreamingUrlSucceeded
  (params (Url :type String) (Code :type Integer) (RequestType :type EEpicMediaRequestType%20Enum :default "UNKNOWN"))
  (bind _mediasource (Variables|Default|GetMediaSource))
  (FortniteSubtitles|SubtitlesPlayer|BindtoMediaPlayer (Variables|Default|GetSubtitlePlayer) (Variables|Default|GetMediaPlayer))
  (Stream|UnbindallEventsfromOnVideoUrlSuccess _mediasource)
  (Stream|UnbindallEventsfromOnVideoUrlFailed _mediasource)
  (Media|MediaPlayer|BindEventtoOnMediaOpenFailed (Variables|Default|GetMediaPlayer) (EventDispatchers|CreateEvent :!event Custom|MediaOpenedFailed))
  (Media|MediaPlayer|BindEventtoOnMediaOpened (Variables|Default|GetMediaPlayer) (EventDispatchers|CreateEvent :!event Custom|MediaOpened))
  (Media|MediaPlayer|OpenSourcewithOptions (Variables|Default|GetMediaPlayer) (Variables|Default|GetMediaSource) (Stream|UpdatePlayerOptions (Variables|Default|GetMediaSource) (select (Variables|/Script/FortniteGame.FortStreamMediaSource|GetbIsLive (Variables|Default|GetMediaSource)) (Utilities|Struct|MakeMediaPlayerOptions "(Audio=0,Caption=-1,Metadata=-1,Script=-1,Subtitle=-1,Text=-1,Video=0)" "(Video=\"\",Audio=\"\",Subtitle=\"\",Caption=\"\")" "UseTrackOptionIndices" (Math|Timespan|FromSeconds (Variables|Default|GetStartPlayTime)) "RelativeToStartTime" "Disabled" "Enabled") (Utilities|Struct|MakeMediaPlayerOptions "(Audio=0,Caption=-1,Metadata=-1,Script=-1,Subtitle=-1,Text=-1,Video=0)" "(Video=\"\",Audio=\"\",Subtitle=\"\",Caption=\"\")" "UseTrackOptionIndices" (Math|Timespan|FromSeconds 1e+05) "RelativeToStartTime" "Disabled" "Disabled"))) :!class /Script/MediaAssets.MediaPlayer))

(event Custom|StreamingUrlFailed
  (params (Url :type String) (Code :type Integer) (RequestType :type EEpicMediaRequestType%20Enum :default "UNKNOWN"))
  (Variables|Default|SetUrlError true)
  (CallFunction|LoadFallback)
  (CallFunction|UnBindAllEvents))

(event Custom|Start%20pip
  (bind _returnvalue (Options|GetRuntimeOptions))
  ;[1 1 1] Start Pip
  (Networking|SwitchHasAuthority
    (:Remote
      (if (Variables|/Script/FortniteGame.FortRuntimeOptions|GetbHiddenButEnabledLiveStreamInMatch (Options|GetRuntimeOptions))
        (Utilities|FlowControl|Delay (Math|Random|RandomFloatinRange 0.0 1.8e+02))
        (Development|PrintString "Start PIP")
        (Variables|/Script/FortniteGame.FortRuntimeOptions|SetbEnableLiveStreamInMatch true _returnvalue)
        (Variables|/Script/FortniteGame.FortRuntimeOptions|SetbShowLiveStreamInMatchByDefault true _returnvalue)))))

(event Custom|PlayResumed
  (Variables|Default|SetMediaRetryCounter 0))

(event Custom|MediaOpened
  (params (NewParam :type String))
  (bind _mediaplayer (Variables|Default|GetMediaPlayer))
  (bind _subtitleplayer (Variables|Default|GetSubtitlePlayer))
  (Variables|Default|SetMediaRetryCounter 0)
  (Media|MediaPlayer|BindEventtoOnMediaClosed (Variables|Default|GetMediaPlayer) (EventDispatchers|CreateEvent :!event Custom|MediaClosed))
  (Media|MediaPlayer|BindEventtoOnEndReached _mediaplayer (EventDispatchers|CreateEvent :!event Custom|MediaEndReached))
  (Media|MediaPlayer|BindEventtoOnPlaybackResumed _mediaplayer (EventDispatchers|CreateEvent :!event Custom|PlayResumed))
  (FortniteSubtitles|SubtitlesPlayer|SetSubtitles _subtitleplayer (Variables|Default|GetSubttleOverlay))
  (CallFunction|StopPIP)
  (FortniteSubtitles|SubtitlesPlayer|Play _subtitleplayer)
  (Media|MediaPlayer|Play (Variables|Default|GetMediaPlayer) :!class /Script/MediaAssets.MediaPlayer))

(event Custom|MediaOpenedFailed
  (params (NewParam :type String))
  (Variables|Default|SetOpenError true)
  (CallFunction|MediaRetry))

(event Custom|Stop%20PIP
  (bind _returnvalue (Options|GetRuntimeOptions))
  ;[1 1 1] Stop Pip
  (Networking|SwitchHasAuthority
    (:Remote
      (if (Variables|/Script/FortniteGame.FortRuntimeOptions|GetbHiddenButEnabledLiveStreamInMatch (Options|GetRuntimeOptions))
        (Variables|/Script/FortniteGame.FortRuntimeOptions|SetbEnableLiveStreamInMatch false _returnvalue)
        (Variables|/Script/FortniteGame.FortRuntimeOptions|SetbShowLiveStreamInMatchByDefault false _returnvalue)))))

(event Custom|MediaClosed
  (FortniteSubtitles|SubtitlesPlayer|Stop (Variables|Default|GetSubtitlePlayer))
  (Variables|Default|SetRunningError true)
  (if (Media|MediaPlayer|HasError (Variables|Default|GetMediaPlayer))
    (CallFunction|MediaRetry)
    (else
      (CallFunction|UnBindAllEvents))))

(event Custom|MediaEndReached
  (FortniteSubtitles|SubtitlesPlayer|Stop (Variables|Default|GetSubtitlePlayer))
  (Variables|Default|SetMediaRetryCounter 0)
  (CallFunction|UnBindAllEvents))

(event Custom|DJBSetupMeshTrigger
  (bind _meshtrigger (Variables|Default|GetMeshTrigger))
  ;[1 1 1] Setup Mesh Trigger
  (bind _outactors (Actor|GetAllActorsOfClass "/Game/Movies/DJBooth/BP_DJBoothMeshTrigger.BP_DJBoothMeshTrigger_C"))
  (if (Utilities|Array|IsValidIndex _outactors)
    (Utilities|IsValid (Utilities|Array|Get(acopy) _outactors)
      (:Is%20Valid
        (bind _output_get (Variables|Default|SetMeshTrigger (Utilities|Array|Get(acopy) _outactors)))
        (Utilities|FlowControl|Sequence
          (:then_0
            (Default|BindEventtoOnReadytoPlay _meshtrigger (delegate Custom|DJBMeshTriggerReady))
            (Default|BindEventtoOnStopPlayback _meshtrigger (delegate Custom|DJBMeshStopPlayback)))
          (:then_1
            (if (not (== (Variables|/Game/Movies/DJBooth/BP_DJBoothMeshTrigger.BP_DJBoothMeshTrigger_C|GetVideoPlayStartTime _output_get) -1.0))
              (Fortnite|FortShippingLog 0 (Utilities|String|Append "DJBooth: Mesh trigger VideoPlayStartTime already set. Play now using fresh play time.
" "ServerWorldTimeSeconds = " (Utilities|String|ToString(Float) (Fortnite|Sync|Time|GetServerWorldTimeSeconds)) "
Replicated PlayServerTimeSeconds: " (Utilities|String|ToString(Float) (Variables|/Game/Movies/DJBooth/BP_DJBoothMeshTrigger.BP_DJBoothMeshTrigger_C|GetPlayServerTimeSeconds (Variables|Default|GetMeshTrigger)))))
              (bind _output_get_2 (Variables|Default|SetStartPlayTime (- (Fortnite|Sync|Time|GetServerWorldTimeSeconds) (Variables|/Game/Movies/DJBooth/BP_DJBoothMeshTrigger.BP_DJBoothMeshTrigger_C|GetPlayServerTimeSeconds (Variables|Default|GetMeshTrigger)))))
              (Fortnite|FortShippingLog 0 (Utilities|String|Append "StartPlayTime= " (Utilities|String|ToString(Float) _output_get_2)))
              (CallFunction|DjbSon (stale Custom|DJBMeshTriggerReady AudioUID)))))))))

(event Custom|UnBindAllEvents
  (bind _mediaplayer (Variables|Default|GetMediaPlayer))
  (Media|MediaPlayer|UnbindallEventsfromOnEndReached _mediaplayer)
  (Media|MediaPlayer|UnbindallEventsfromOnMediaClosed _mediaplayer)
  (Media|MediaPlayer|UnbindallEventsfromOnMediaOpenFailed _mediaplayer)
  (Media|MediaPlayer|UnbindallEventsfromOnMediaOpened _mediaplayer)
  (Media|MediaPlayer|UnbindallEventsfromOnPlaybackResumed _mediaplayer))

(event Custom|DJBMeshTriggerReady
  (params (PlayStartTime :type Float%20%28double-precision%29) (AudioUID :type String))
  ;[1 1 1] Setup Mesh Trigger
  (Variables|Default|SetStartPlayTime PlayStartTime)
  (CallFunction|DjbSon AudioUID))

(event Custom|DJBMeshStopPlayback
  ;[1 1 1] Setup Mesh Trigger
  (CallFunction|DjbSoff))

(event Custom|LoadFallback
  (CallFunction|UnBindAllEvents)
  (Audio|Components|Audio|Play (Variables|Default|GetAudio) 0.0))

(event Custom|Media%20Retry
  (if (<= (Variables|Default|GetMediaRetryCounter) (Variables|Default|GetMediaRetryMax))
    (Math|Integer|IncrementInt (Variables|Default|GetMediaRetryCounter)
      (:%20%20
        (CallFunction|RequestStreamURL)))
    (else
      (CallFunction|LoadFallback))))

(fn ConstructionScript)
